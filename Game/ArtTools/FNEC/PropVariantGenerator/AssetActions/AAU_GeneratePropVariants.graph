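# Editor utility asset action that generates fixed-color variants from selected RandomColor_Prop Blueprints.
# It reads each prop's color swatches from its class default object, then duplicates the Blueprint per swatch into a StaticVariants subfolder.
# GetTokensFromBaseAsset builds each name by dropping single-character tokens, appending a variant suffix and swapping BP_ for FNEC_.
# Accumulated copies are reparented to BuildingProp, given swatch texture data by array index, compiled and saved.
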EditorUtilityBlueprint /Game/ArtTools/FNEC/PropVariantGenerator/AssetActions/AAU_GeneratePropVariants
graphs [
 CreateVariantsFromRandomizedProp {nodes [
   CreateVariantsFromRandomizedProp() {is=title K2Node_FunctionEntry}
   > n1 = GetSelectedAssets()
   > n2 = For Each Loop(Array:Array of Object References=@n1) -> (Array Element:Object Reference)
   > n3 = Sequence!(execute) -> (then_0, then_1, then_2:Exec)
   > n4 = Set RCP_Asset(RCP_Asset:Object Reference=@n2) -> (Output_Get:Object Reference)
   > n5 = GetBlueprintAsset(Object=@n4)
   > n6 = GeneratedClass!(execute, BlueprintObj:Blueprint Object Reference=@n5) -> (then, ReturnValue:Object Class Reference)
   > n7 = GetClassDefaultObject!(execute, Class:Object Class Reference=@n6) -> (then, ReturnValue:Object Reference)
   > n8 = Cast To RandomColor_Prop!(execute, Object=@n7) -> (then, CastFailed, AsRandom Color Prop:Random Color Prop Object Reference) {is=title K2Node_DynamicCast}
   > n9 = GetSwatchArrayConst!(execute, self:Random Color Prop Object Reference=@n8) -> (then, Color Swatches:Array of Building Texture Data Object References, Swatch Count:int=0)
   > Set AR_ColorSwatches(AR_ColorSwatches:Array of Building Texture Data Object References=@n9) -> (Output_Get:Array of Building Texture Data Object References)
   > Set NumberOfChildren(NumberOfChildren:int=@n9.Swatch Count) -> (Output_Get:int=0)
   n10 = Get NumberOfChildren() -> (NumberOfChildren:int=0)
   Branch(execute=@n3.then_1, Condition="integer > integer"(A=@n10))
   > n11 = For Loop(LastIndex=int - int(A=@n10))
   > n12 = Set VariantNumber(VariantNumber:int=@n11) -> (Output_Get:int=0)
   > Set TextureData(TextureData:Building Texture Data Object Reference="Get (a copy)"(
     Array:Array of Building Texture Data Object References=Get AR_ColorSwatches() -> (AR_ColorSwatches:Array of Building Texture Data Object References)
     Dimension 1=@n12
    ) -> (Output:Building Texture Data Object Reference)) -> (Output_Get:Building Texture Data Object Reference)
   > n13 = GetTokensFromBaseAsset!(execute, self) -> (then, NewAssetPath:String)
   > PrintString(InString=Append(A=NEWPATH_, B=@n13))
   > n14 = DuplicateLoadedAsset!(execute, SourceAsset:Object Reference=Get RCP_Asset() -> (RCP_Asset:Object Reference), DestinationAssetPath:String=@n13) -> (then, ReturnValue:Object Reference)
   > Add Unique(
    TargetArray:Array of Object References=Get AR_NewAssets() -> (AR_NewAssets:Array of Object References)
    NewItem:"Object Reference (by ref)"=@n14
   )
   n15 = For Each Loop(Exec=@n3.then_2, Array=Get AR_NewAssets())
   SaveLoadedAssets(execute=@n15.Completed, AssetsToSave=Get AR_NewAssets(), bOnlyIfIsDirty=false)
   n16 = Set L_AssetToModify(execute=@n15.LoopBody, L_AssetToModify:Object Reference=@n15) -> (Output_Get:Object Reference)
   > n17 = GetBlueprintAsset(Object=@n16)
   > ReparentBlueprint(Blueprint=@n17, NewParentClass=&BuildingProp)
   > CompileBlueprint!(execute, Blueprint:Blueprint Object Reference=@n17) -> (then, ReturnValue:bool)
   > n18 = GetBlueprintAsset(Object=Get L_AssetToModify() -> (L_AssetToModify:Object Reference))
   > n19 = GeneratedClass(BlueprintObj=@n18)
   > n20 = GetClassDefaultObject(Class=@n19)
   > n21 = Cast To BuildingProp!(execute, Object=@n20) -> (then, CastFailed, AsBuilding Prop:Building Prop Object Reference) {is=title K2Node_DynamicCast}
   > SetBuildingTextureData!(
    execute
    self:Building Static Mesh Object Reference=@n21
    Index:int=0
    NewBuildingTextureData:Building Texture Data Object Reference="Get (a copy)"(Array=Get AR_ColorSwatches(), Dimension 1=@n15.Array Index)
   ) -> (then)
   > CompileBlueprint(Blueprint=@n18)
  ]}
 EventGraph {links [], nodes []}
 GetTokensFromBaseAsset {nodes [
   GetTokensFromBaseAsset() {is=title K2Node_FunctionEntry}
   > n1 = Sequence!(execute) -> (then_0, then_1)
   > n2 = GetPathNameForLoadedAsset(LoadedAsset=Get RCP_Asset())
   n3 = Split(SourceString=@n2, InStr=., SearchDir=FromEnd)
   > Set L_BaseName(L_BaseName:String=@n3.RightS) -> (Output_Get:String)
   > Set L_BasePath(L_BasePath:String=Split(SourceString=@n3, InStr=/, SearchDir=FromEnd)) -> (Output_Get:String)
   > n4 = For Each Loop(Array:Array of Strings=ParseIntoArray(SourceString=Get L_BaseName() -> (L_BaseName:String), Delimiter=_)) -> (Array Element:String)
   > Branch(Condition="integer > integer"(A=Len(S=@n4), B=1))
   > Add Unique(TargetArray:Array of Strings=Get L_AR_SplitStrings() -> (L_AR_SplitStrings:Array of Strings), NewItem:"String (by ref)"=@n4)
   n5 = JoinStringArray(SourceArray=Get L_AR_SplitStrings(), Separator=_)
   n6 = "Get (a copy)"(
    Array:Array of Strings=Get L_AR_AlphaNumeric() -> (L_AR_AlphaNumeric:Array of Strings)
    Dimension 1=Get VariantNumber() -> (VariantNumber:int=0)
   ) -> (Output:String)
   Set NewAssetName(execute=@n1.then_1, NewAssetName:String=Replace(SourceString=Append!(A=@n5, B=_, C:String=@n6) -> (ReturnValue), From=BP_, To=FNEC_)) -> (Output_Get:String)
   n7 = Get NewAssetName() -> (NewAssetName:String)
   > Set L_NewAssetPath(L_NewAssetPath:String=Append!(A=Get L_BasePath() -> (L_BasePath:String), B=/, C=StaticVariants, D:String=/, E:String=@n7, F:String=., G:String=@n7) -> (ReturnValue)) -> (Output_Get:String)
   > Return Node!(execute, NewAssetPath:String=Get L_NewAssetPath() -> (L_NewAssetPath:String))
  ]}
]
variables [
 AR_ColorSwatches {type="array<BuildingTextureData>", default []}
 AR_NewAssets {type="array<Object>", default []}
 AR_Props_RandomColor {type="array<Object>", default []}
 NewAssetName {type=string, default=""}
 NumberOfChildren {type=integer}
 RCP_Asset {type=Object, default=null}
 SwatchData {type=Class, default=null}
 TextureData {type=BuildingTextureData, default=null}
 VariantNumber {type=integer}
]
